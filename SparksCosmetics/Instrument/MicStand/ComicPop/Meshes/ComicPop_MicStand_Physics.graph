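# Physics asset for the ComicPop mic stand, a Sparks instrument cosmetic, defining collision and joints for its hanging-link dynamics.
# Two capsules cover the handle, small spheres mark the left and right screws, and one capsule covers each of ten links.
# Two five-link chains run from the screws with limited swing and mostly limited twist.
# A fully angle-locked joint between dyn_link_05_connector and dyn_link_10 closes them into one loop.
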PhysicsAsset /SparksCosmetics/Instrument/MicStand/ComicPop/Meshes/ComicPop_MicStand_Physics
bodies {
 dyn_link_01 [{shapeType=Capsule, length=2, radius=1, shapeName=dyn_link_01_capsule, center=<2 0 0>, rotation=<rot 90 0 0>}]
 dyn_link_02 [{shapeType=Capsule, length=2, radius=1.3, shapeName=dyn_link_02_capsule, center=<2.5 0 0>, rotation=<rot 90 0 0>}]
 dyn_link_03 [{shapeType=Capsule, length=2.5, radius=1.25, shapeName=dyn_link_03_capsule, center=<2.45 0 0>, rotation=<rot 90 0 0>}]
 dyn_link_04 [{shapeType=Capsule, length=2.0657, radius=1.2115, shapeName=dyn_link_04_capsule, center=<1.9855 -0.0162 -0.5>, rotation=<rot 90 180 -175>}]
 dyn_link_05_connector [{shapeType=Capsule, length=1, radius=1, shapeName=dyn_link_05_connector_capsule, center=<0.7 0.6 0.8>, rotation=<rot 41.65 -180 150>}]
 dyn_link_06 [{shapeType=Capsule, length=3, radius=1, shapeName=dyn_link_06_capsule, center=<1.77 0 0>, rotation=<rot 90 0 0>}]
 dyn_link_07 [{shapeType=Capsule, length=2.6, radius=1.37, shapeName=dyn_link_07_capsule, center=<2.2 0 0>, rotation=<rot 90 0 0>}]
 dyn_link_08 [{shapeType=Capsule, length=3.26, radius=1.2398, shapeName=dyn_link_08_capsule, center=<2.05 0.06 0>, rotation=<rot 90 0 0>}]
 dyn_link_09 [{shapeType=Capsule, length=2.0966, radius=1.3934, shapeName=dyn_link_09_capsule, center=<1.83 0 -0.42>, rotation=<rot 90 0 0>}]
 dyn_link_10 [{shapeType=Capsule, length=1, radius=1, shapeName=dyn_link_05_connector_capsule, center=<0.77 -0.7 0.76>, rotation=<rot 49.4 -166.5 -115>}]
 dyn_screw_l [{shapeType=Sphere, radius=0.5}]
 dyn_screw_r [{shapeType=Sphere, radius=0.5}]
 handle [
  {shapeType=Capsule, length=121.4574, radius=1.4495, shapeName=micstand main, center=<0 0 6.1647>}
  {shapeType=Capsule, length=6.4718, radius=1.9208, shapeName=top, center=<0 0.0785 71.6484>}
 ]
}
constraints [
 {bone1Name=dyn_link_01, bone2Name=dyn_screw_r, swing1LimitDegrees=90, swing2LimitDegrees=45, twistLimitDegrees=10}
 {bone1Name=dyn_link_02, bone2Name=dyn_link_01, swing1LimitDegrees=55, swing2LimitDegrees=45, twistLimitDegrees=10}
 {bone1Name=dyn_link_03, bone2Name=dyn_link_02, swing1LimitDegrees=60, swing2LimitDegrees=45, twistLimitDegrees=10}
 {bone1Name=dyn_link_04, bone2Name=dyn_link_03, swing1LimitDegrees=60, swing2LimitDegrees=50, twistLimitDegrees=10}
 {
  bone1Name=dyn_link_05_connector
  bone2Name=dyn_link_04
  swing1LimitDegrees=20
  swing2LimitDegrees=20
  twistLimitDegrees=5
  twistMotion=Locked
 }
 {bone1Name=dyn_link_06, bone2Name=dyn_screw_l, swing1LimitDegrees=90, swing2LimitDegrees=45, twistLimitDegrees=5, twistMotion=Locked}
 {bone1Name=dyn_link_07, bone2Name=dyn_link_06, swing1LimitDegrees=55, swing2LimitDegrees=45, twistLimitDegrees=10}
 {bone1Name=dyn_link_08, bone2Name=dyn_link_07, swing1LimitDegrees=60, swing2LimitDegrees=45, twistLimitDegrees=10}
 {bone1Name=dyn_link_09, bone2Name=dyn_link_08, swing1LimitDegrees=60, swing2LimitDegrees=50, twistLimitDegrees=10}
 {bone1Name=dyn_link_10, bone2Name=dyn_link_09, swing1LimitDegrees=20, swing2LimitDegrees=20, twistLimitDegrees=10}
 {
  bone1Name=dyn_link_05_connector
  bone2Name=dyn_link_10
  swing1LimitDegrees=45
  swing1Motion=Locked
  swing2LimitDegrees=45
  swing2Motion=Locked
  twistLimitDegrees=45
  twistMotion=Locked
 }
]
